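# Unlit translucent particle sprite material that draws fireflies from a SubUV texture.
# The red channel is tinted by particle color, while the green channel masks a boosted cyan-green glow scaled by the first dynamic parameter, which is off by default.
# Both layers are summed through MF_FX_EmissiveAdjustments, and texture alpha is multiplied by particle alpha for transparency.
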Material /Game/Packages/Fortress_Transylvania/Transylvania_effects/Materials/M_Firefly
graph {
 properties {
  bUsedWithParticleSprites=true
  blendMode=BLEND_Translucent
  pixelDepthOffsetMode=PDOM_Legacy
  shadingModel=MSM_Unlit
  usedShadingModels=Unlit
 }
 nodes [
  n1 = ParticleSubUV!(UVs, Tex, Apply View MipBias) {texture=/Game/Packages/Fortress_Transylvania/Transylvania_effects/Textures/T_FireFly}
  n2 = ParticleColor()
  n3 = Multiply(A=@n1.G, B=Multiply(A=Constant3Vector() {constant=<rgba 0 3.5 2.7952 0>}) {constB=3})
  n4 = DynamicParameter() {defaultValue=<rgba 0 0 0 1>}
  MaterialFunctionCall(Emissive=Add(A=Multiply(A=@n1.R, B=@n2.RGB), B=Multiply(A=@n3, B=@n4.Param1))) {materialFunction=&MF_FX_EmissiveAdjustments}
  Multiply(A=@n1.A, B=@n2.A)
 ]
}
