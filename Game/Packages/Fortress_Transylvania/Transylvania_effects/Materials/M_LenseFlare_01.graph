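# M_LenseFlare_01 is a translucent lens flare material for Niagara and Cascade particle sprites in the Transylvania fortress effects.
# Particle color RGB feeds an emissive adjustment function.
# Particle alpha times the flare texture's green channel forms the opacity, which a static switch, on by default, scales from zero using Distance_Blend.
# The clamped result passes through DepthFade to soften intersections with scene geometry.
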Material /Game/Packages/Fortress_Transylvania/Transylvania_effects/Materials/M_LenseFlare_01
graph {
 properties {
  bUsedWithNiagaraSprites=true
  bUsedWithParticleSprites=true
  blendMode=BLEND_Translucent
  pixelDepthOffsetMode=PDOM_Legacy
  usedShadingModels=Default Lit
 }
 nodes [
  n1 = ParticleColor()
  n2 = TextureSampleParameter2D() {parameterName=FlareTexture, samplerType=SAMPLERTYPE_Grayscale, texture=/Game/Effects/Fort_Effects/Textures/T_EFX_CameraFlare_01}
  n3 = Multiply(A=@n1.A, B=@n2.G)
  n4 = LinearInterpolate(
   A=Multiply(A=@n3) {constB=0}
   B=@n3
   Alpha=MaterialFunctionCall!(Blend Range=Constant() {r=500}, Start Offset=Constant() {r=-100}) {materialFunction=/Engine/Functions/Engine_MaterialFunctions03/Particles/Distance_Blend}.Result
  )
  n5 = StaticBoolParameter() {defaultValue=true, parameterName=Use Distance Blend}
  DepthFade(Opacity=Clamp(None=StaticSwitch(True=@n4, False=@n3, Value=@n5))) {fadeDistanceDefault=50}
  MaterialFunctionCall(Emissive=@n1.RGB) {materialFunction=&MF_FX_EmissiveAdjustments}
 ]
}
