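# Additive, unlit particle material for sprite and Niagara effects that draws a torch flame from a sub-UV fire flipbook.
# Static switches choose plain flipbook color, a desaturated version, or panning fluid-flame noise masked by the flipbook.
# Dynamic parameters drive noise tiling, pan speed over particle lifetime, erosion and camera-fade blend.
# Particle color tints the boosted emissive, while particle alpha feeds depth fade with optional camera-distance fade.
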Material /Game/Packages/Fortress_Transylvania/Transylvania_effects/Materials/M_Torch_Flame_01
graph {
 properties {
  bUsedWithNiagaraSprites=true
  bUsedWithParticleSprites=true
  blendMode=BLEND_Additive
  pixelDepthOffsetMode=PDOM_Legacy
  shadingModel=MSM_Unlit
  usedShadingModels=Unlit
 }
 nodes [
  LinearInterpolate()
  n1 = ParticleSubUV!(UVs, Tex, Apply View MipBias) {texture=/Game/Effects/Fort_Effects/Textures/T_Fire_Ball_Sub_UV3}
  n2 = Multiply(A=TextureCoordinate() {vTiling=0.5}, B=ScalarParameter() {defaultValue=4, parameterName=Param_1})
  n3 = DynamicParameter() {paramNames [Tiling, Erode Detail, Random Speed, CameraDepthFade]}
  n4 = ParticleRelativeTime()
  n5 = TextureSample!(
   UVs=Panner(Coordinate=Multiply(A=@n2, B=@n3.Tiling), Time=Multiply(A=@n3.Random Speed, B=@n4)) {speedY=1.5}
   Tex
   Level=ScalarParameter() {defaultValue=4}
   Apply View MipBias
  ) {
   mipValueMode=TMVM_MipLevel
   samplerType=SAMPLERTYPE_LinearGrayscale
   texture=/Game/Weapons/FORT_Melee/Pickaxe_Cube_Ninja_Male/FX/T_FluidFlameTile
  }
  n6 = Subtract(A=@n5.G, B=DynamicParameter() {paramNames [Tiling, Erode Detail, Random Speed, CameraDepthFade]}.Erode Detail)
  n7 = StaticSwitchParameter(
   True=Desaturation(None=@n1.RGB, Fraction=ScalarParameter() {parameterName=Desat})
   False=StaticSwitchParameter(True=LinearInterpolate(B=@n6, Alpha=@n1.G), False=@n1.RGB) {parameterName=Add Detail Noise}
  ) {parameterName=Desaturate}
  n8 = ParticleColor()
  n9 = DepthFade(Opacity=Multiply(A=@n7, B=@n8.A), FadeDistance=ScalarParameter() {defaultValue=30, parameterName=DepthFade_Distance}) {fadeDistanceDefault=30}
  n10 = ScalarParameter() {defaultValue=50, parameterName=CameraDepthFade_Length}
  n11 = ScalarParameter() {parameterName=CameraDepthFade_Offset}
  LinearInterpolate(
   A=@n9
   B=Multiply(
    A=@n9
    B=MaterialFunctionCall(Fade Length=@n10, Fade Offset=@n11) {materialFunction=&CameraDepthFade}.Result
   )
   Alpha=Multiply(
    A=ScalarParameter() {parameterName=Switch_CameraDepthFade}
    B=DynamicParameter() {paramNames [Tiling, Erode Detail, Random Speed, CameraDepthFade]}.CameraDepthFade
   )
  )
  MaterialFunctionCall(Emissive=Multiply(A=Multiply(A=@n8.RGB, B=@n7)) {constB=4}) {materialFunction=&MF_FX_EmissiveAdjustments}
  Power()
 ]
}
