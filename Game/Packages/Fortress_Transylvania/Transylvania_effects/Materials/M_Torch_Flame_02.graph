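# Translucent, unlit, two-sided particle material for a torch flame in the Fortress Transylvania effects package.
# A SubUV flame flipbook is tinted by particle color for emissive, boosted and passed through MF_FX_EmissiveAdjustments.
# Flipbook RGB times particle alpha feeds a depth fade for soft intersections.
# Both paths have night/day lerps driven by FortniteActiveTimeOfDay, but identical endpoint multipliers leave the look unchanged by time of day.
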Material /Game/Packages/Fortress_Transylvania/Transylvania_effects/Materials/M_Torch_Flame_02
graph {
 properties {
  bUsedWithNiagaraSprites=true
  bUsedWithParticleSprites=true
  blendMode=BLEND_Translucent
  pixelDepthOffsetMode=PDOM_Legacy
  shadingModel=MSM_Unlit
  twoSided=true
  usedShadingModels=Unlit
 }
 nodes [
  Constant() {r=1}
  n1 = ParticleSubUV!(UVs, Tex, Apply View MipBias) {texture=/Game/Packages/Fortress_Transylvania/Transylvania_effects/Textures/T_flame_flipbook_01}
  n2 = ParticleColor()
  n3 = Multiply(A=@n1.RGB, B=@n2.A)
  n4 = CollectionParameter() {
   collection=&FortniteMaterialParameters
   parameterName=FortniteActiveTimeOfDay
  }
  n5 = DotProduct(A=@n4, B=Constant4Vector() {constant=<rgba 1 1 1 0>}.RGBA)
  DepthFade(Opacity=LinearInterpolate(A=Multiply(A=@n3) {constB=0.5}, B=Multiply(A=@n3) {constB=0.5}, Alpha=@n5) {desc="A=Night B=Day"}) {fadeDistanceDefault=30}
  n6 = Multiply(A=@n2.RGB, B=@n1.RGB)
  MaterialFunctionCall(Emissive=LinearInterpolate(A=Multiply(A=@n6) {constB=20}, B=Multiply(A=@n6) {constB=20}, Alpha=@n5) {desc="A=Night B=Day"}) {materialFunction=&MF_FX_EmissiveAdjustments}
  Multiply(B=Multiply(A=Constant() {r=1.4}))
 ]
}
