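# Translucent, lit mesh-particle material for animated water foam, tinted and faded by particle color.
# Two panning water displacement samples, warped by a scrolling distortion map, form a stepped foam pattern shaped by DynamicParameter inputs.
# Opacity combines that pattern with a V-gradient edge mask and depth fade.
# Panning water normals offset scene color for fake refraction, damped toward screen edges, desaturated and blended into emissive.
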Material /Game/Packages/Fortress_Transylvania/Transylvania_effects/Materials/M_Water_Foam_1
graph {
 properties {
  bTangentSpaceNormal=false
  bUsedWithMeshParticles=true
  blendMode=BLEND_Translucent
  numCustomizedUVs=1
  opacityMaskClipValue=0.999
  pixelDepthOffsetMode=PDOM_Legacy
  translucencyLightingMode=TLM_VolumetricPerVertexNonDirectional
  translucentShadowDensityScale=2
  translucentShadowStartOffset=5
  usedShadingModels=Default Lit
 }
 nodes [
  n1 = TextureSample(UVs=Panner(Coordinate=TextureCoordinate() {uTiling=2, vTiling=2}) {speedY=0.4}) {texture=/Game/Global/Textures/Default/Tileable_Patterns/T_Tilling_Distortion_Map_01}
  n2 = Multiply(A=ComponentMask(None=@n1.G) {g=true, r=true}) {constB=0.1}
  n3 = Panner(Coordinate=Add(A=@n2, B=TextureCoordinate() {uTiling=2})) {speedX=0.32, speedY=0.4}
  n4 = Panner(Coordinate=Add(A=@n2, B=TextureCoordinate() {uTiling=6, vTiling=4})) {speedX=-0.2, speedY=1.2}
  n5 = Power(Base=TextureSample(UVs=@n4) {texture=/Game/Packages/DS_Fortnite_TRV_Terrain/Textures/T_TRV_HF_WaterDisp}.R, Exp=Constant() {r=1})
  n6 = LinearInterpolate(A=TextureSample(UVs=@n3) {texture=&T_TRV_HF_WaterDisp}.R, B=@n5, Alpha=@n5)
  n7 = Divide(A=Ceil(None=Multiply(A=@n6) {constB=10})) {constB=5}
  n8 = Clamp(None=Multiply(A=@n6, B=Clamp(None=@n7)))
  n9 = DynamicParameter() {defaultValue=<rgba 0 0 0 1>, paramNames [Param1,Param2,VTiling,UTiling]}
  n10 = Power(Base=Multiply(A=@n8, B=@n1.G), Exp=LinearInterpolate(Alpha=@n9.Param2) {constA=5, constAlpha=0}) {constExponent=50}
  n11 = LinearInterpolate(Alpha=@n9.Param2) {constAlpha=0, constB=200}
  n12 = Multiply(A=@n8, B=Clamp(None=Multiply(A=@n10, B=@n11)))
  n13 = MaterialFunctionCall() {materialFunction=&LinearGradient}
  n14 = Power(Base=@n13.VGradient) {constExponent=3}
  n15 = Power(Base=OneMinus(None=@n13.VGradient)) {constExponent=1.2}
  n16 = Multiply(A=Clamp(None=Multiply(A=@n14) {constB=5}), B=Clamp(None=Multiply(A=@n15) {constB=2}))
  n17 = Clamp(None=@n16)
  n18 = LinearInterpolate(Alpha=@n17) {constA=3, constAlpha=0}
  n19 = Clamp(None=Multiply(A=Power(Base=@n8, Exp=@n18) {constExponent=50}, B=LinearInterpolate(Alpha=@n17) {constAlpha=0, constB=200}))
  n20 = ParticleColor()
  DepthFade(Opacity=Multiply(A=Multiply(A=@n12, B=@n19), B=@n20.A) {constB=3}) {fadeDistanceDefault=50}
  n21 = Add(
   A=TextureSample(UVs=@n3) {samplerType=SAMPLERTYPE_Normal, texture=/Game/Packages/DS_Fortnite_TRV_Terrain/Textures/T_TRV_HF_WaterNormal}.RGB
   B=TextureSample(UVs=@n4) {samplerType=SAMPLERTYPE_Normal, texture=&T_TRV_HF_WaterNormal}.RGB
  )
  n22 = Power(Base=Clamp(None=OneMinus(None=@n16))) {constExponent=35}
  n23 = ComponentMask(None=LinearInterpolate(A=Clamp(None=@n21), B=Constant3Vector() {constant=<rgba 0 0 1 0>}, Alpha=Clamp(None=@n22)) {constA=1}) {g=true, r=true}
  n24 = ViewSize()
  n25 = SceneTexelSize()
  n26 = ConstantBiasScale(None=Divide(A=ScreenPosition().ViewportUV, B=Multiply(A=@n24, B=@n25))) {bias=-0.5, scale=2}
  n27 = OneMinus(None=Multiply(A=@n26, B=@n26))
  n28 = MaterialFunctionCall(Float2=Clamp(None=Multiply(A=@n27) {constB=0.5})) {materialFunction=&BreakOutFloat2Components}
  n29 = Add(
   A=ComponentMask(None=ScreenPosition().ViewportUV) {g=true, r=true}
   B=Multiply(A=Multiply(A=@n23) {constB=0.5}, B=Multiply(A=@n28.R, B=@n28.R))
  )
  n30 = SceneColor!(UVs=@n29)
  n31 = Clamp(None=Desaturation(
    None=ShadingPathSwitch(Default=@n30.RGB, Deferred=@n30.RGB, Mobile=@n29) {desc=Migrated from FeatureLevelSwitch}
    Fraction=Constant() {r=0.25}
   ))
  n32 = Multiply(A=@n8) {constA=1, constB=4}
  n33 = Power(Base=OneMinus(None=LinearInterpolate(A=@n32, Alpha=@n9.Param1) {constAlpha=0.04})) {constExponent=60}
  n34 = Add(A=Clamp(None=OneMinus(None=@n8)), B=Clamp(None=Multiply(A=@n33) {constB=50}))
  n35 = Multiply(
   A=Power(Base=@n1.R, Exp=LinearInterpolate(Alpha=@n9.VTiling) {constA=15, constAlpha=0}) {constExponent=50}
   B=LinearInterpolate(Alpha=@n9.VTiling) {constAlpha=0, constB=5}
  )
  n36 = Clamp(None=Add(A=Multiply(A=@n34), B=Clamp(None=@n35)) {constB=0.6})
  MaterialFunctionCall(Emissive=Add(A=LinearInterpolate(B=@n31, Alpha=@n36) {constA=0.5, constB=0}) {constB=0.1}) {materialFunction=&MF_FX_EmissiveAdjustments}
  MaterialFunctionCall(DiffuseColor=Multiply(A=@n36, B=@n20.RGB)) {materialFunction=&ConvertFromDiffSpec}
 ]
}
