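# Cheap translucent particle-sprite material for waterfall splash effects.
# Particle color RGB feeds the MF_FX_EmissiveAdjustments function, while the splash texture's alpha is multiplied by particle alpha.
# Cost stays low with a single texture sample and per-vertex non-directional volumetric lighting; translucent shadow densities are zeroed.
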Material /Game/Packages/Fortress_Transylvania/Transylvania_effects/Materials/M_waterfall_splash_Cheap_01
graph {
 properties {
  bTangentSpaceNormal=false
  bUsedWithParticleSprites=true
  blendMode=BLEND_Translucent
  pixelDepthOffsetMode=PDOM_Legacy
  translucencyLightingMode=TLM_VolumetricPerVertexNonDirectional
  translucencyPass=MTP_BeforeDOF
  translucentSelfShadowDensityScale=0
  translucentSelfShadowSecondDensityScale=0
  translucentShadowDensityScale=0
  usedShadingModels=Default Lit
 }
 nodes [
  Constant()
  MaterialFunctionCall(Emissive=ParticleColor().RGB) {materialFunction=&MF_FX_EmissiveAdjustments}
  Multiply(
   A=TextureSample(UVs=TextureCoordinate()) {desc=o, texture=/Game/Effects/Fort_Effects/Textures/Water/T_water_splash_04}.A
   B=ParticleColor().A
  )
 ]
}
